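# Anim Blueprint for the FermiumStraw Rain female character's face accessory mesh, adding secondary motion to poses copied from other meshes.
# A cached per-bone blend of two copied poses feeds head-space RigidBody physics, a Control Rig, then fixed additive offsets on five hair base bones.
# Each update computes GravityOverride from a parameter asset, but the visible AnimGraph never reads it, and RigidBody gravity shows a constant.
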AnimBlueprint /BRCosmetics/Characters/Player/Female/Medium/Bodies/F_MED_FermiumStraw_Rain/Meshes/Parts/F_MED_FermiumStraw_Rain_FaceAcc_AnimBP
target_skeleton=&/F_MED_FermiumStraw_Rain_FaceAcc_Skeleton
graphs [
 AnimGraph {nodes [
   n1 = RigidBody!(ComponentPose=Local To Component(LocalPose=Use cached pose 'Start Cache'()), Alpha, OverrideWorldGravity:Vector="0,0,-980", ExternalForce) -> (Pose) {
    bOverrideWorldGravity=true
    overridePhysicsAsset=&/F_MED_FermiumStraw_Rain_FaceAcc_Physics
    simulationSpace=BaseBoneSpace
    alphaBoolBlend {blendOption=Linear}
    baseBoneRef {boneName=head}
    componentAppliedLinearAccClamp=<2500 2500 2500>
    componentLinearAccScale=<0.35 0.35 0.35>
    componentLinearVelScale=<0.35 0.35 0.35>
    simSpaceSettings {maxAngularAcceleration=10, maxAngularVelocity=10, maxLinearAcceleration=10, maxLinearVelocity=10, worldAlpha=1}
   }
   n2 = Control Rig(Source=Component To Local(ComponentPose=@n1)) {
    alphaBoolBlend {blendOption=Linear}
    controlRigAssetReference {blueprintRigClass=&/F_MED_FermiumStraw_Rain_FaceAcc_CtrlRig.F_MED_FermiumStraw_Rain_FaceAcc_CtrlRig_C}
   }
   n3 = "Transform (Modify) Bone - Bone: dyn_hair_01_front_base_l"(ComponentPose=Local To Component(LocalPose=@n2), Alpha=1, Translation="-0.5,0,0", Rotation="6,0,0", Scale="1,1,1") {
    rotationMode=BMM_Additive
    rotationSpace=BCS_BoneSpace
    translationMode=BMM_Additive
    translationSpace=BCS_BoneSpace
    boneToModify {boneName=dyn_hair_01_front_base_l}
    rotation {pitch=6}
    translation {x=-0.5}
   }
   n4 = "Transform (Modify) Bone - Bone: dyn_hair_01_side_base_l"(ComponentPose=@n3, Alpha=1, Translation="-1,0,0", Rotation="6,0,0", Scale="1,1,1") {
    rotationMode=BMM_Additive
    rotationSpace=BCS_BoneSpace
    translationMode=BMM_Additive
    translationSpace=BCS_BoneSpace
    boneToModify {boneName=dyn_hair_01_side_base_l}
    rotation {pitch=6}
    translation {x=-1}
   }
   n5 = "Transform (Modify) Bone - Bone: dyn_hair_01_back_base_c"(ComponentPose=@n4, Alpha=1, Translation="-0.5,0,0", Rotation="12,0,0", Scale="1,1,1") {
    rotationMode=BMM_Additive
    rotationSpace=BCS_BoneSpace
    translationMode=BMM_Additive
    translationSpace=BCS_BoneSpace
    boneToModify {boneName=dyn_hair_01_back_base_c}
    rotation {pitch=12}
    translation {x=-0.5}
   }
   n6 = "Transform (Modify) Bone - Bone: dyn_hair_01_side_base_r"(ComponentPose=@n5, Alpha=1, Translation="-1,0,0", Rotation="12,0,0", Scale="1,1,1") {
    rotationMode=BMM_Additive
    rotationSpace=BCS_BoneSpace
    translationMode=BMM_Additive
    translationSpace=BCS_BoneSpace
    boneToModify {boneName=dyn_hair_01_side_base_r}
    rotation {pitch=12}
    translation {x=-1}
   }
   n7 = "Transform (Modify) Bone - Bone: dyn_hair_01_front_base_r"(ComponentPose=@n6, Alpha=1, Translation="-0.5,0,0", Rotation="6,0,0", Scale="1,1,1") {
    rotationMode=BMM_Additive
    rotationSpace=BCS_BoneSpace
    translationMode=BMM_Additive
    translationSpace=BCS_BoneSpace
    boneToModify {boneName=dyn_hair_01_front_base_r}
    rotation {pitch=6}
    translation {x=-0.5}
   }
   Output Pose(Result=Component To Local(ComponentPose=@n7)) {layerGroup=None}
   n8 = Get SkelMesh()
   Save cached pose 'Start Cache'(Pose=Layered blend per bone(
     BasePose=Copy Pose From Mesh() {bUseAttachedParent=true}
     BlendPoses_0=Copy Pose From Mesh(SourceMeshComponent=@n8) {bUseAttachedParent=true}
    )) {cacheName=Start Cache}
  ]}
 EventGraph {nodes [
   Event BlueprintUpdateAnimation()
   > n1 = "Gravity Override State Function (Parameter Asset) (Nativized)"(
    CharacterPartAnimInstance=Self-Reference()
    GameTimeInSeconds=GetGameTimeInSeconds()
    Parameters=&/F_MED_FermiumStraw_Rain_Parameters
   )
   > Set GravityOverride(GravityOverride=@n1)
   Event BlueprintInitializeAnimation()
   > Set SkelMesh(SkelMesh=GetOwnerSkeletalMeshComponent())
  ]}
]
variables [
 GravityOverride {type=Vector}
 SkelMesh {type=SkeletalMeshComponent}
]
